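# Physics asset for the PolarityWinn Coil female medium player body cosmetic, setting up secondary motion for its front flap, back flap and keychain.
# Capsules on the pelvis, spine and thighs form a simple body volume, linked by free-motion constraints.
# Back and front flap chains hang from master bones locked to spine_03 and spine_04, with angular limits along each chain and minimal twist.
# The keychain uses a two-bone limited chain off the pelvis.
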PhysicsAsset /BRCosmetics/Characters/Player/Female/Medium/Bodies/F_MED_PolarityWinn_Coil/Meshes/F_MED_PolarityWinn_Coil_Physics
bodies {
 dyn_bk_flap_01 [{
   shapeType=Capsule
   length=2.4555
   radius=2.009
   shapeName=dyn_bk_flap_01_capsule
   center=<4.676 -0.0061 -0.2801>
   rotation=<rot -90 -101.119 -85.9225>
  }]
 dyn_bk_flap_02 [{
   shapeType=Capsule
   length=2.7099
   radius=1.7546
   shapeName=dyn_bk_flap_02_capsule
   center=<4.676 -0.0061 -0.5628>
   rotation=<rot -90 -99.4789 -87.5626>
  }]
 dyn_bk_flap_03 [{
   shapeType=Capsule
   length=3.7061
   radius=1.5835
   shapeName=dyn_bk_flap_03_capsule
   center=<4.676 -0.0061 0.1178>
   rotation=<rot -90 -120.9797 -66.0617>
  }]
 dyn_bk_flap_04 [{
   shapeType=Capsule
   length=4.8608
   radius=1.5735
   shapeName=dyn_bk_flap_04_capsule
   center=<1.4722 0.0958 0.0176>
   rotation=<rot -90 -114.7965 -72.2449>
  }]
 dyn_bk_flap_MSTR [{shapeType=Capsule, length=1, radius=1, shapeName=dyn_bk_flap_MSTR_capsule, rotation=<rot -90 0 0>}]
 dyn_fr_flap_01 [{shapeType=Capsule, length=1.4604, radius=2.5396, shapeName=dyn_fr_flap_01_capsule, center=<2.25 0 0.0378>, rotation=<rot -89.038 0 0>}]
 dyn_fr_flap_02 [{shapeType=Capsule, length=1.4604, radius=2.5396, shapeName=dyn_fr_flap_02_capsule, rotation=<rot -89.038 0 0>}]
 dyn_fr_flap_MSTR [{shapeType=Capsule, length=0.505, radius=0.505, shapeName=dyn_fr_flap_MSTR_capsule, rotation=<rot 0 -90 0>}]
 dyn_keychain [
  {
   shapeType=Capsule
   length=5.242
   radius=2.4498
   shapeName=dyn_keychain_capsule
   center=<6.0943 -0.894 -0.0455>
   rotation=<rot 0 -98.2779 -3.2514>
  }
  {shapeType=Capsule, length=4.9741, radius=1.2046, shapeName=dyn_keychain_capsule, center=<2.1445 0 0>}
 ]
 dyn_keychain_MSTR [{shapeType=Capsule, length=0.505, radius=0.505, shapeName=dyn_keychain_MSTR_capsule, rotation=<rot 0 -90 0>}]
 pelvis [{shapeType=Capsule, length=11.7796, radius=9.7489, shapeName=pelvis_capsule, center=<-4.827 -2.47 0>}]
 spine_01 [{shapeType=Capsule, length=8.0525, radius=8.1657, shapeName=spine_01_capsule, center=<-1.903 -0.3246 0>, rotation=<rot 0 -90 0>}]
 spine_02 [{shapeType=Capsule, length=6.7123, radius=8.2877, shapeName=spine_02_capsule, center=<0 3.3851 0>}]
 spine_03 [{shapeType=Capsule, length=9.6709, radius=7.7299, shapeName=spine_03_capsule, center=<0 2.2317 0>}]
 spine_04 [{shapeType=Capsule, length=5.713, radius=9.287, shapeName=spine_04_capsule}]
 thigh_l [{
   shapeType=Capsule
   length=17.8927
   radius=7.9541
   shapeName=thigh_l_capsule
   center=<-16.6579 -1.8441 -0.101>
   rotation=<rot 0 78.7934 -86.2398>
  }]
 thigh_r [{
   shapeType=Capsule
   length=24.1724
   radius=8.224
   shapeName=thigh_r_capsule
   center=<18.6056 0.1285 0.3009>
   rotation=<rot 0.0789 -94.1163 -89.3646>
  }]
}
constraints [
 {
  bone1Name=spine_01
  bone2Name=pelvis
  swing1LimitDegrees=45
  swing1Motion=Free
  swing2LimitDegrees=45
  swing2Motion=Free
  twistLimitDegrees=45
  twistMotion=Free
 }
 {
  bone1Name=spine_02
  bone2Name=spine_01
  swing1LimitDegrees=45
  swing1Motion=Free
  swing2LimitDegrees=45
  swing2Motion=Free
  twistLimitDegrees=45
  twistMotion=Free
 }
 {
  bone1Name=spine_03
  bone2Name=spine_02
  swing1LimitDegrees=45
  swing1Motion=Free
  swing2LimitDegrees=45
  swing2Motion=Free
  twistLimitDegrees=45
  twistMotion=Free
 }
 {bone1Name=dyn_bk_flap_01, bone2Name=dyn_bk_flap_MSTR, swing1LimitDegrees=20, swing2LimitDegrees=20, twistLimitDegrees=1}
 {bone1Name=dyn_bk_flap_02, bone2Name=dyn_bk_flap_01, swing1LimitDegrees=30, swing2LimitDegrees=30, twistLimitDegrees=1}
 {bone1Name=dyn_bk_flap_03, bone2Name=dyn_bk_flap_02, swing1LimitDegrees=45, swing2LimitDegrees=45, twistLimitDegrees=1}
 {bone1Name=dyn_bk_flap_04, bone2Name=dyn_bk_flap_03, swing1LimitDegrees=45, swing2LimitDegrees=45, twistLimitDegrees=1}
 {
  bone1Name=dyn_bk_flap_MSTR
  bone2Name=spine_03
  swing1LimitDegrees=30
  swing1Motion=Locked
  swing2LimitDegrees=20
  swing2Motion=Locked
  twistLimitDegrees=1
  twistMotion=Locked
 }
 {
  bone1Name=spine_04
  bone2Name=spine_03
  swing1LimitDegrees=45
  swing1Motion=Free
  swing2LimitDegrees=45
  swing2Motion=Free
  twistLimitDegrees=45
  twistMotion=Free
 }
 {
  bone1Name=thigh_l
  bone2Name=pelvis
  swing1LimitDegrees=45
  swing1Motion=Free
  swing2LimitDegrees=45
  swing2Motion=Free
  twistLimitDegrees=45
  twistMotion=Free
 }
 {
  bone1Name=thigh_r
  bone2Name=pelvis
  swing1LimitDegrees=45
  swing1Motion=Free
  swing2LimitDegrees=45
  swing2Motion=Free
  twistLimitDegrees=45
  twistMotion=Free
 }
 {bone1Name=dyn_fr_flap_01, bone2Name=dyn_fr_flap_MSTR, swing1LimitDegrees=45, swing2LimitDegrees=1, twistLimitDegrees=1}
 {bone1Name=dyn_fr_flap_02, bone2Name=dyn_fr_flap_01, swing1LimitDegrees=45, swing2LimitDegrees=1, twistLimitDegrees=1}
 {
  bone1Name=dyn_fr_flap_MSTR
  bone2Name=spine_04
  swing1LimitDegrees=45
  swing1Motion=Locked
  swing2LimitDegrees=45
  swing2Motion=Locked
  twistLimitDegrees=45
  twistMotion=Locked
 }
 {bone1Name=dyn_keychain_MSTR, bone2Name=pelvis, swing1LimitDegrees=60, swing2LimitDegrees=45, twistLimitDegrees=1}
 {bone1Name=dyn_keychain, bone2Name=dyn_keychain_MSTR, swing1LimitDegrees=60, swing2LimitDegrees=30, twistLimitDegrees=10}
]
