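# Animation Blueprint for the Polarity Winn Coil female body cosmetic, adding secondary motion to flaps, keychain and pants.
# It layers the head mesh's copied pose over the parent mesh's, then runs a Control Rig.
# bRigidBodyEnabled selects RigidBody simulation or an AnimDynamics chain for flaps and keychain, followed by AnimDynamics for pants in both cases.
# Skydive diving or floating switches RigidBody gravity to point upward.
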AnimBlueprint /BRCosmetics/Characters/Player/Female/Medium/Bodies/F_MED_PolarityWinn_Coil/Meshes/F_MED_PolarityWinn_Coil_AnimBP
target_skeleton=&/F_MED_PolarityWinn_Coil_Skeleton
graphs [
 AnimGraph {nodes [
   n1 = Control Rig(Source=Use cached pose 'null_pose'()) {
    alphaBoolBlend {blendOption=Linear}
    controlRigAssetReference {blueprintRigClass=&/F_MED_PolarityWinn_Coil_CtrlRig.F_MED_PolarityWinn_Coil_CtrlRig_C}
   }
   n2 = RigidBody!(ComponentPose=Local To Component(LocalPose=@n1), Alpha, OverrideWorldGravity:Vector=Get Gravity Override(), ExternalForce) -> (Pose) {
    bClampLinearTranslationLimitToRefPose=true
    bOverrideWorldGravity=true
    overridePhysicsAsset=&/F_MED_PolarityWinn_Coil_Physics
    simulationSpace=BaseBoneSpace
    alphaBoolBlend {blendOption=Linear}
    baseBoneRef {boneName=pelvis}
    overrideWorldGravity {z=-980}
    simSpaceSettings {dampingAlpha=0, maxAngularVelocity=600, maxLinearVelocity=500, velocityScaleZ=1.5, worldAlpha=0.35}
   }
   n3 = Component To Local(ComponentPose=@n2)
   n4 = Control Rig(Source=Use cached pose 'null_pose'()) {
    alphaBoolBlend {blendOption=Linear}
    controlRigAssetReference {blueprintRigClass=&F_MED_PolarityWinn_Coil_CtrlRig_C}
   }
   n5 = Anim Dynamics - Bone: dyn_fr_flap_01(ComponentPose=Local To Component(LocalPose=@n4), Alpha=1, ExternalForce="0,0,0") {
    angularBiasOverride=0.999
    angularDampingOverride=0.999
    angularSpringConstant=600
    bAngularSpring=true
    bOverrideAngularDamping=true
    bUseGravityOverride=true
    simulationSpace=World
    alphaBoolBlend {blendOption=Linear}
    boundBone {boneName=dyn_fr_flap_01}
    componentAppliedLinearAccClamp=<2000 2000 2000>
    componentLinearAccScale=<0.1 0.1 0.1>
    componentLinearVelScale=<0.1 0.1 0.1>
    gravityOverride {z=-980}
    physicsBodyDefinitions [{
      boundBone {boneName=dyn_fr_flap_01}
      boxExtents=<15 15 15>
      constraintSetup {angularLimitsMax=<0 0 90>, angularLimitsMin=<0 0 -10>, angularTarget=<0 0 0>}
      localJointOffset=<1 0 -2>
     }]
   }
   n6 = Anim Dynamics - Bone: dyn_fr_flap_02(ComponentPose=@n5, Alpha=1, ExternalForce="0,0,0") {
    angularBiasOverride=0.999
    angularDampingOverride=0.999
    angularSpringConstant=600
    bAngularSpring=true
    bOverrideAngularDamping=true
    bUseGravityOverride=true
    simulationSpace=World
    alphaBoolBlend {blendOption=Linear}
    boundBone {boneName=dyn_fr_flap_02}
    componentAppliedLinearAccClamp=<2000 2000 2000>
    componentLinearAccScale=<0.1 0.1 0.1>
    componentLinearVelScale=<0.1 0.1 0.1>
    gravityOverride {z=-980}
    physicsBodyDefinitions [{
      boundBone {boneName=dyn_fr_flap_02}
      boxExtents=<10 10 10>
      constraintSetup {angularLimitsMax=<0 0 90>, angularLimitsMin=<0 0 -10>, angularTarget=<0 0 0>}
      localJointOffset=<1 0 -2>
     }]
   }
   n7 = Anim Dynamics - Bone: dyn_bk_flap_01(ComponentPose=@n6, Alpha=1, ExternalForce="0,0,0") {
    angularBiasOverride=0.999
    angularDampingOverride=0.999
    angularSpringConstant=600
    bAngularSpring=true
    bOverrideAngularBias=true
    bOverrideAngularDamping=true
    bUseGravityOverride=true
    simulationSpace=World
    alphaBoolBlend {blendOption=Linear}
    boundBone {boneName=dyn_bk_flap_01}
    componentAppliedLinearAccClamp=<2000 2000 2000>
    componentLinearAccScale=<0.1 0.1 0.1>
    componentLinearVelScale=<0.1 0.1 0.1>
    gravityOverride {z=-980}
    physicsBodyDefinitions [{
      boundBone {boneName=dyn_bk_flap_01}
      boxExtents=<10 10 10>
      constraintSetup {angularLimitsMax=<0 30 15>, angularLimitsMin=<0 -5 -15>, angularTarget=<0 0 0>}
      localJointOffset=<1 0 0>
     }]
   }
   n8 = Anim Dynamics - Bone: dyn_bk_flap_02(ComponentPose=@n7, Alpha=1, ExternalForce="0,0,0") {
    angularBiasOverride=0.999
    angularDampingOverride=0.999
    angularSpringConstant=600
    bAngularSpring=true
    bOverrideAngularBias=true
    bOverrideAngularDamping=true
    bUseGravityOverride=true
    bUseSphericalLimits=true
    simulationSpace=World
    alphaBoolBlend {blendOption=Linear}
    boundBone {boneName=dyn_bk_flap_02}
    componentAppliedLinearAccClamp=<2000 2000 2000>
    componentLinearAccScale=<0.1 0.1 0.1>
    componentLinearVelScale=<0.1 0.1 0.1>
    gravityOverride {z=-980}
    physicsBodyDefinitions [{
      boundBone {boneName=dyn_bk_flap_02}
      boxExtents=<10 10 10>
      constraintSetup {angularLimitsMax=<0 30 15>, angularLimitsMin=<0 -10 -15>, angularTarget=<0 0 0>}
      localJointOffset=<1 0 0>
     }]
    sphericalLimits [{limitRadius=10, drivingBone {boneName=pelvis}, sphereLocalOffset=<6.8125 -2.0603 3.9473>}]
   }
   n9 = Anim Dynamics - Bone: dyn_bk_flap_03(ComponentPose=@n8, Alpha=1, ExternalForce="0,0,0") {
    angularBiasOverride=0.999
    angularDampingOverride=0.999
    angularSpringConstant=600
    bAngularSpring=true
    bOverrideAngularBias=true
    bOverrideAngularDamping=true
    bUseGravityOverride=true
    bUseSphericalLimits=true
    simulationSpace=World
    alphaBoolBlend {blendOption=Linear}
    boundBone {boneName=dyn_bk_flap_03}
    componentAppliedLinearAccClamp=<2000 2000 2000>
    componentLinearAccScale=<0.1 0.1 0.1>
    componentLinearVelScale=<0.1 0.1 0.1>
    gravityOverride {z=-980}
    physicsBodyDefinitions [{
      boundBone {boneName=dyn_bk_flap_03}
      boxExtents=<10 10 10>
      constraintSetup {angularLimitsMax=<0 30 15>, angularLimitsMin=<0 -10 -15>, angularTarget=<0 0 0>}
      localJointOffset=<1 0 0>
     }]
    sphericalLimits [{limitRadius=11.2711, drivingBone {boneName=pelvis}, sphereLocalOffset=<-0.0063 -2.8294 3.5767>}]
   }
   n10 = Anim Dynamics - Bone: dyn_bk_flap_04(ComponentPose=@n9, Alpha=1, ExternalForce="0,0,0") {
    angularBiasOverride=0.999
    angularDampingOverride=0.999
    angularSpringConstant=600
    bAngularSpring=true
    bOverrideAngularBias=true
    bOverrideAngularDamping=true
    bUseGravityOverride=true
    bUseSphericalLimits=true
    simulationSpace=World
    alphaBoolBlend {blendOption=Linear}
    boundBone {boneName=dyn_bk_flap_04}
    componentAppliedLinearAccClamp=<2000 2000 2000>
    componentLinearAccScale=<0.1 0.1 0.1>
    componentLinearVelScale=<0.1 0.1 0.1>
    gravityOverride {z=-980}
    physicsBodyDefinitions [{
      boundBone {boneName=dyn_bk_flap_04}
      boxExtents=<10 10 10>
      constraintSetup {angularLimitsMax=<0 30 15>, angularLimitsMin=<0 -10 -15>, angularTarget=<0 0 0>}
      localJointOffset=<1 0 0>
     }]
    sphericalLimits [{limitRadius=10.6672, drivingBone {boneName=pelvis}, sphereLocalOffset=<-4.586 -3.2829 -0.0004>}]
   }
   n11 = Anim Dynamics - Bone: dyn_keychain(ComponentPose=@n10, Alpha=1, ExternalForce="0,0,0") {
    angularBiasOverride=0.999
    angularDampingOverride=0.999
    angularSpringConstant=600
    bAngularSpring=true
    bOverrideAngularDamping=true
    bUseGravityOverride=true
    bUseSphericalLimits=true
    simulationSpace=BoneRelative
    alphaBoolBlend {blendOption=Linear}
    boundBone {boneName=dyn_keychain}
    componentAppliedLinearAccClamp=<2000 2000 2000>
    componentLinearAccScale=<0.1 0.1 0.1>
    componentLinearVelScale=<0.1 0.1 0.1>
    gravityOverride {z=-980}
    physicsBodyDefinitions [{
      boundBone {boneName=dyn_keychain}
      boxExtents=<10 10 10>
      constraintSetup {angularLimitsMax=<10 30 -7>, angularLimitsMin=<-10 -10 -90>, angularTarget=<0 0 0>}
      localJointOffset=<1 0 -2>
     }]
    planarLimits [{drivingBone {boneName=thigh_r}, planeTransform {location=<1.2186 -7.9765 -3.4403>, rotation=<rot 21.2412 -9.86 -89.7022>}}]
    sphericalLimits [{limitRadius=11.4686, drivingBone {boneName=thigh_r}, sphereLocalOffset=<0.3338 3.374 -3.3174>}]
   }
   n12 = Component To Local(ComponentPose=@n11)
   n13 = Get bRigidBodyEnabled()
   n14 = Anim Dynamics - Bone: dyn_pant_l_bk(
    ComponentPose=Local To Component(LocalPose=Blend Poses by bool(BlendPose_0=@n3, BlendPose_1=@n12, bActiveValue=@n13) {blendType=Linear})
    Alpha=1
    ExternalForce="0,0,0"
   ) {
    angularBiasOverride=0.999
    angularDampingOverride=0.999
    angularSpringConstant=600
    bAngularSpring=true
    bOverrideAngularDamping=true
    bUseGravityOverride=true
    simulationSpace=World
    alphaBoolBlend {blendOption=Linear}
    boundBone {boneName=dyn_pant_l_bk}
    componentAppliedLinearAccClamp=<2000 2000 2000>
    componentLinearAccScale=<0.3 0.3 0.3>
    componentLinearVelScale=<0.3 0.3 0.3>
    gravityOverride {z=-980}
    physicsBodyDefinitions [{
      boundBone {boneName=dyn_pant_l_bk}
      boxExtents=<10 10 10>
      constraintSetup {angularLimitsMax=<0 3 3>, angularLimitsMin=<0 -3 -3>, angularTarget=<0 0 0>}
      localJointOffset=<1 0 0>
     }]
   }
   n15 = Anim Dynamics - Bone: dyn_pant_l_fr(ComponentPose=@n14, Alpha=1, ExternalForce="0,0,0") {
    angularBiasOverride=0.999
    angularDampingOverride=0.999
    angularSpringConstant=600
    bAngularSpring=true
    bOverrideAngularDamping=true
    bUseGravityOverride=true
    bUseSphericalLimits=true
    simulationSpace=World
    alphaBoolBlend {blendOption=Linear}
    boundBone {boneName=dyn_pant_l_fr}
    componentAppliedLinearAccClamp=<2000 2000 2000>
    componentLinearAccScale=<0.3 0.3 0.3>
    componentLinearVelScale=<0.3 0.3 0.3>
    gravityOverride {z=-980}
    physicsBodyDefinitions [{
      boundBone {boneName=dyn_pant_l_fr}
      boxExtents=<10 10 10>
      constraintSetup {angularLimitsMax=<0 3 3>, angularLimitsMin=<0 -3 -3>, angularTarget=<0 0 0>}
      localJointOffset=<1 0 0>
     }]
    sphericalLimits [{limitRadius=10, drivingBone {boneName=pelvis}, sphereLocalOffset=<6.7945 0.6667 3.946>}]
   }
   n16 = Anim Dynamics - Bone: dyn_pant_l_outer(ComponentPose=@n15, Alpha=1, ExternalForce="0,0,0") {
    angularBiasOverride=0.999
    angularDampingOverride=0.999
    angularSpringConstant=600
    bAngularSpring=true
    bOverrideAngularDamping=true
    bUseGravityOverride=true
    bUseSphericalLimits=true
    simulationSpace=World
    alphaBoolBlend {blendOption=Linear}
    boundBone {boneName=dyn_pant_l_outer}
    componentAppliedLinearAccClamp=<2000 2000 2000>
    componentLinearAccScale=<0.3 0.3 0.3>
    componentLinearVelScale=<0.3 0.3 0.3>
    gravityOverride {z=-980}
    physicsBodyDefinitions [{
      boundBone {boneName=dyn_pant_l_outer}
      boxExtents=<10 10 10>
      constraintSetup {angularLimitsMax=<0 3 3>, angularLimitsMin=<0 -3 -3>, angularTarget=<0 0 0>}
      localJointOffset=<1 0 0>
     }]
    sphericalLimits [{limitRadius=11.2711, drivingBone {boneName=pelvis}, sphereLocalOffset=<0 -1.1913 3.5784>}]
   }
   n17 = Anim Dynamics - Bone: dyn_pant_l_inner(ComponentPose=@n16, Alpha=1, ExternalForce="0,0,0") {
    angularBiasOverride=0.999
    angularDampingOverride=0.999
    angularSpringConstant=600
    bAngularSpring=true
    bOverrideAngularDamping=true
    bUseGravityOverride=true
    bUseSphericalLimits=true
    simulationSpace=World
    alphaBoolBlend {blendOption=Linear}
    boundBone {boneName=dyn_pant_l_inner}
    componentAppliedLinearAccClamp=<2000 2000 2000>
    componentLinearAccScale=<0.3 0.3 0.3>
    componentLinearVelScale=<0.3 0.3 0.3>
    gravityOverride {z=-980}
    physicsBodyDefinitions [{
      boundBone {boneName=dyn_pant_l_inner}
      boxExtents=<10 10 10>
      constraintSetup {angularLimitsMax=<0 3 3>, angularLimitsMin=<0 -3 -3>, angularTarget=<0 0 0>}
      localJointOffset=<1 0 0>
     }]
    sphericalLimits [{limitRadius=10.6672, drivingBone {boneName=pelvis}, sphereLocalOffset=<-4.5982 -2.4889 -0.0017>}]
   }
   n18 = Anim Dynamics - Bone: dyn_pant_r_bk(ComponentPose=@n17, Alpha=1, ExternalForce="0,0,0") {
    angularBiasOverride=0.999
    angularDampingOverride=0.999
    angularSpringConstant=600
    bAngularSpring=true
    bOverrideAngularDamping=true
    bUseGravityOverride=true
    simulationSpace=World
    alphaBoolBlend {blendOption=Linear}
    boundBone {boneName=dyn_pant_r_bk}
    componentAppliedLinearAccClamp=<2000 2000 2000>
    componentLinearAccScale=<0.3 0.3 0.3>
    componentLinearVelScale=<0.3 0.3 0.3>
    gravityOverride {z=-980}
    physicsBodyDefinitions [{
      boundBone {boneName=dyn_pant_r_bk}
      boxExtents=<10 10 10>
      constraintSetup {angularLimitsMax=<0 3 3>, angularLimitsMin=<0 -3 -3>, angularTarget=<0 0 0>}
      localJointOffset=<-1 0 0>
     }]
   }
   n19 = Anim Dynamics - Bone: dyn_pant_r_fr(ComponentPose=@n18, Alpha=1, ExternalForce="0,0,0") {
    angularBiasOverride=0.999
    angularDampingOverride=0.999
    angularSpringConstant=600
    bAngularSpring=true
    bOverrideAngularDamping=true
    bUseGravityOverride=true
    bUseSphericalLimits=true
    simulationSpace=World
    alphaBoolBlend {blendOption=Linear}
    boundBone {boneName=dyn_pant_r_fr}
    componentAppliedLinearAccClamp=<2000 2000 2000>
    componentLinearAccScale=<0.3 0.3 0.3>
    componentLinearVelScale=<0.3 0.3 0.3>
    gravityOverride {z=-980}
    physicsBodyDefinitions [{
      boundBone {boneName=dyn_pant_r_fr}
      boxExtents=<10 10 10>
      constraintSetup {angularLimitsMax=<0 3 3>, angularLimitsMin=<0 -3 -3>, angularTarget=<0 0 0>}
      localJointOffset=<-1 0 0>
     }]
    sphericalLimits [{limitRadius=10, drivingBone {boneName=pelvis}, sphereLocalOffset=<6.7945 0.6667 3.946>}]
   }
   n20 = Anim Dynamics - Bone: dyn_pant_r_outer(ComponentPose=@n19, Alpha=1, ExternalForce="0,0,0") {
    angularBiasOverride=0.999
    angularDampingOverride=0.999
    angularSpringConstant=600
    bAngularSpring=true
    bOverrideAngularDamping=true
    bUseGravityOverride=true
    bUseSphericalLimits=true
    simulationSpace=World
    alphaBoolBlend {blendOption=Linear}
    boundBone {boneName=dyn_pant_r_outer}
    componentAppliedLinearAccClamp=<2000 2000 2000>
    componentLinearAccScale=<0.3 0.3 0.3>
    componentLinearVelScale=<0.3 0.3 0.3>
    gravityOverride {z=-980}
    physicsBodyDefinitions [{
      boundBone {boneName=dyn_pant_r_outer}
      boxExtents=<10 10 10>
      constraintSetup {angularLimitsMax=<0 3 3>, angularLimitsMin=<0 -3 -3>, angularTarget=<0 0 0>}
      localJointOffset=<-1 0 0>
     }]
    sphericalLimits [{limitRadius=11.2711, drivingBone {boneName=pelvis}, sphereLocalOffset=<0 -1.1913 3.5784>}]
   }
   n21 = Anim Dynamics - Bone: dyn_pant_r_inner(ComponentPose=@n20, Alpha=1, ExternalForce="0,0,0") {
    angularBiasOverride=0.999
    angularDampingOverride=0.999
    angularSpringConstant=600
    bAngularSpring=true
    bOverrideAngularDamping=true
    bUseGravityOverride=true
    bUseSphericalLimits=true
    simulationSpace=World
    alphaBoolBlend {blendOption=Linear}
    boundBone {boneName=dyn_pant_r_inner}
    componentAppliedLinearAccClamp=<2000 2000 2000>
    componentLinearAccScale=<0.3 0.3 0.3>
    componentLinearVelScale=<0.3 0.3 0.3>
    gravityOverride {z=-980}
    physicsBodyDefinitions [{
      boundBone {boneName=dyn_pant_r_inner}
      boxExtents=<10 10 10>
      constraintSetup {angularLimitsMax=<0 3 3>, angularLimitsMin=<0 -3 -3>, angularTarget=<0 0 0>}
      localJointOffset=<-1 0 0>
     }]
    sphericalLimits [{limitRadius=10.6672, drivingBone {boneName=pelvis}, sphereLocalOffset=<-4.5982 -2.4889 -0.0017>}]
   }
   Output Pose(Result=Component To Local(ComponentPose=@n21)) {layerGroup=None}
   n22 = Get Head_Mesh() -> (Head_Mesh:Skeletal Mesh Component Budgeted Object Reference)
   Save cached pose 'null_pose'(Pose=Layered blend per bone(
     BasePose=Copy Pose From Mesh() {bUseAttachedParent=true}
     BlendPoses_0=Copy Pose From Mesh(SourceMeshComponent=@n22) {bUseAttachedParent=true}
    )) {cacheName=null_pose}
  ]}
 EventGraph {nodes [
   Event BlueprintUpdateAnimation()
   n1 = Get bIsSkydiveDiveMode()
   n2 = Get bIsSkydiveFloating()
   > Set Gravity Override(Gravity Override=SelectVector(A="0,0,500", B="0,0,-980", bPickA=OR Boolean(A=@n1, B=@n2)))
   n3 = Event OnCharacterCustomizationLoaded()
   > Set Head_Mesh(Head_Mesh:Skeletal Mesh Component Budgeted Object Reference=Get Array Item with EFort Custom Part Type(Array=@n3.CustomMeshParts, PartType=Head)) -> (Output_Get:Skeletal Mesh Component Budgeted Object Reference)
  ]}
]
variables [
 Game Time in Seconds {type=number}
 Gravity Override {type=Vector}
 Gravity Override Params {type=GravityOverrideParamsStruct}
 Head_Mesh {type=SkeletalMeshComponentBudgeted}
]
